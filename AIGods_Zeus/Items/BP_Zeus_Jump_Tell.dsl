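; Telegraph actor for Zeus's leap attack, marking an area with a ground lightning effect and an electric-field sound.
; On BeginPlay, only non-authority instances spawn the sound at the actor's location and store it.
; A collision sphere adds a looping lightning-tell GameplayCue to overlapping FortPawns and removes it on end overlap.
; The DisableTell event deactivates the Niagara telegraph, fades out any stored sound and destroys the actor.
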(event EventBeginPlay
  (Networking|SwitchHasAuthority
    (:Remote
      (bind _returnvalue (Audio|SpawnSoundatLocation "/SunRoseZeusGameplay/Sounds/Boss/SunRose_ZeusBoss_LeapAttack_ElecField_Start_Cue.SunRose_ZeusBoss_LeapAttack_ElecField_Start_Cue" (Transformation|GetActorLocation self)))
      (Variables|Default|SetTellSound _returnvalue))))

(event Collision|EventActorBeginOverlap
  (params (OtherActor :type Actor%20Object%20Reference)))

(event EventTick
  (params (DeltaSeconds :type Float%20%28single-precision%29)))

(event Custom|DisableTell
  (Components|Activation|SetActive (Variables|Default|GetNS_Lightning_Telegraph_OnGround) false false)
  (bind _tellsound (Variables|Default|GetTellSound)
    (:then
      (Audio|Components|Audio|FadeOut _tellsound 0.5 0.0 "Linear" :!class /Script/Engine.AudioComponent))
    (:else
      _))
  (Actor|DestroyActor))

(event OnComponentBeginOverlap(SphereCollision)
  (params (OverlappedComponent :type Primitive%20Component%20Object%20Reference) (OtherActor :type Actor%20Object%20Reference) (OtherComp :type Primitive%20Component%20Object%20Reference) (OtherBodyIndex :type Integer) (bFromSweep :type Boolean) (SweepResult :type Hit%20Result%20Structure%20%28by%20ref%29))
  (bind _asfort_pawn (Utilities|Casting|CastToFortPawn OtherActor))
  (GameplayCue|AddGameplayCueOnActor(Looping) _asfort_pawn (Variables|Default|GetGC_LightningTell) (Ability|GameplayCue|MakeGameplayCueParameters 0.0)))

(event OnComponentEndOverlap(SphereCollision)
  (params (OverlappedComponent :type Primitive%20Component%20Object%20Reference) (OtherActor :type Actor%20Object%20Reference) (OtherComp :type Primitive%20Component%20Object%20Reference) (OtherBodyIndex :type Integer))
  (bind _asfort_pawn (Utilities|Casting|CastToFortPawn OtherActor))
  (GameplayCue|RemoveGameplayCueOnActor(Looping) _asfort_pawn (Variables|Default|GetGC_LightningTell) (Ability|GameplayCue|MakeGameplayCueParameters 0.0)))

(fn ConstructionScript)
