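; Weapon Blueprint for the lightning bolt wielded by the Zeus AI, handling how the bolt appears in hand.
; After the weapon is equipped and attached, it waits briefly, then swaps the weapon mesh to the SunRose LightningBolt gadget skeletal mesh.
; It then makes the right mesh visible and activates the right-side ambient spear Niagara effect.
; BeginPlay and the construction script only call the parent, and Tick is disabled.
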(event EventBeginPlay
  (|CallParentFunction))

(event EventTick
  (params (DeltaSeconds :type Float%20%28single-precision%29))
  (|CallParentFunction DeltaSeconds) ; This node is disabled and will not be called.\nDrag off pins to build functionality.
)

(event FortWeapon|EventOnEquippedAndAttached
  (Utilities|FlowControl|Delay 1.6)
  (Components|SkeletalMesh|SetSkeletalMeshAsset (Variables|Visual|GetWeaponMesh) "/SunRoseZeusGameplay/Items/Lightning/Meshes/LightningBolt_SunRose_Gadget.LightningBolt_SunRose_Gadget")
  (CallFunction|SetRightMeshIsVisible true)
  (Components|Activation|SetActive (Variables|Default|GetNS_SunRose_Zeus_Spear_Ambient_Right) true false))

(fn ConstructionScript
  (|CallParentFunction))
