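; Zeus lightning bolt projectile that explodes when it hits something or enters water.
; A hit, or first water-body entry with splash effects, runs a DoOnce-guarded DoExplosion, then hides the actor and sets a short lifespan.
; The explosion event acts only on the non-authority side, playing Niagara FX, impact sound, camera shake and force feedback, then destroying the Niagara component after a delay.
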(event OnEnterWaterEvent(WaterInteractionComponent)
  (params (WaterBody :type Fort%20Water%20Body%20Actor%20Object%20Reference) (WaterInteractionComponent :type Fort%20Water%20Interaction%20Component%20Object%20Reference) (bIsFirstBody :type Boolean))
  ;[1 1 1] On Entered Water
  (if bIsFirstBody
    (Effects|Components|ParticleSystem|SpawnFXAtLocation "/Game/Athena/Items/Consumables/MolotovCocktail/NS_FireflyJar_WaterSplash.NS_FireflyJar_WaterSplash" (Transformation|GetWorldLocation (Variables|Collision|GetCapsuleComponent)))
    (Audio|PlaySoundatLocation (Variables|Default|GetSplash%20Sound) (Transformation|GetActorLocation))
    ;[1 1 1] On Hit
    (Utilities|FlowControl|DoOnce false
      (:Completed
        (Delivery|DoExplosion)
        (Rendering|SetActorHiddenInGame true)
        (Actor|SetLifeSpan 3.0)))))

(event Collision|EventHit
  (params (MyComp :type Primitive%20Component%20Object%20Reference) (Other :type Actor%20Object%20Reference) (OtherComp :type Primitive%20Component%20Object%20Reference) (bSelfMoved :type Boolean) (HitLocation :type Vector :default "0, 0, 0") (HitNormal :type Vector :default "0, 0, 0") (NormalImpulse :type Vector :default "0, 0, 0") (Hit :type Hit%20Result%20Structure%20%28by%20ref%29))
  ;[1 1 1] On Hit
  (|CallParentFunction MyComp Other OtherComp bSelfMoved HitLocation HitNormal NormalImpulse Hit)
  (Utilities|FlowControl|DoOnce false
    (:Completed
      (Delivery|DoExplosion)
      (Rendering|SetActorHiddenInGame true)
      (Actor|SetLifeSpan 3.0))))

(event EventOnExploded
  (params (HitActors :type Array%20of%20Actor%20Object%20References) (HitResults :type Array%20of%20Hit%20Result%20Structures))
  (bind _returnvalue (Transformation|GetActorLocation))
  (Networking|SwitchHasAuthority
    (:Remote
      (bind _returnvalue_2 (Niagara|SpawnSystematLocation "/SunRoseZeusGameplay/Items/Lightning/FX/System/NS_SunRose_Zeus_Explosion_Optimize.NS_SunRose_Zeus_Explosion_Optimize" (Transformation|GetActorLocation)))
      (Audio|PlaySoundatLocation "/SunRoseZeusGameplay/Sounds/Boss/SunRose_ZeusBoss_Proj_Impact_Large_Cue.SunRose_ZeusBoss_Proj_Impact_Large_Cue" _returnvalue)
      (Rendering|SetActorHiddenInGame true)
      (Camera|PlayWorldCameraShake (Variables|Default|GetBoltImpact_CamShake) _returnvalue 0.0 1.5e+03)
      (ForceFeedback|SpawnForceFeedbackatLocation "/Game/Weapons/ForceFeedback/Explosions/FF_Generic_Explosion_Near.FF_Generic_Explosion_Near" _returnvalue)
      (Utilities|FlowControl|Delay 2.0)
      (Components|DestroyComponent _returnvalue_2))))

(fn ConstructionScript)
